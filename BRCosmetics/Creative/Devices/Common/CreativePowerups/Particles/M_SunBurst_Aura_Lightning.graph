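# Translucent unlit, two-sided material for lightning strands on Niagara mesh particles in the Creative powerup SunBurst aura.
# A gradient mask, randomly flipped per particle and driven by dynamic parameters (with fallbacks at zero particle time), shapes SplineThicken width and fades opacity.
# Opacity also combines particle alpha, distance-intensity modulation and a panning plasma texture or switchable band.
# World position offset adds sine jiggle, scaled by Noise Intensity, to the thickening.
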Material /BRCosmetics/Creative/Devices/Common/CreativePowerups/Particles/M_SunBurst_Aura_Lightning
graph {
 properties {
  bUsedWithNiagaraMeshParticles=true
  blendMode=BLEND_Translucent
  pixelDepthOffsetMode=PDOM_Legacy
  shadingModel=MSM_Unlit
  twoSided=true
  usedShadingModels=Unlit
 }
 nodes [
  n1 = Time()
  Frac(None=@n1)
  n2 = DynamicParameter() {paramNames [Width, Mask Size, Mask Position, Noise Intensity]}
  n3 = Ceil(None=ParticleRelativeTime() {desc=">0"})
  n4 = ComponentMask(None=TextureCoordinate()) {g=true}
  n5 = OneMinus(None=@n4)
  n6 = Round(None=ParticleRandom())
  n7 = Time()
  n8 = MaterialFunctionCall!(
   Gradient=LinearInterpolate(A=@n4, B=LinearInterpolate(A=@n4, B=@n5, Alpha=@n6), Alpha=ScalarParameter() {defaultValue=1, parameterName=UseParticleRandomFlip})
   Mask Position=LinearInterpolate(A=Frac(None=@n7), B=@n2.Mask Position, Alpha=@n3) {constA=0.5}
   Mask Size=LinearInterpolate(A=Constant() {r=3}, B=@n2.Mask Size, Alpha=@n3)
  ) {materialFunction=/Game/Effects/Fort_Effects/MaterialFunctions/MF_GradientMask}
  n9 = Multiply(A=LinearInterpolate(A=Constant() {r=1.333}, B=@n2.Width, Alpha=@n3), B=@n8.Sinusoidal)
  n10 = VertexInterpolator(VS=Power(
    Base=MaterialFunctionCall!(Width=@n9) {materialFunction=/BRCosmetics/Weapons/FORT_Melee/Pickaxe_Parfait_Female/FX/Materials/Functions/MF_TrifleParfait_SplineThicken}.Distance Intensity Modulation
    Exp=ScalarParameter() {defaultValue=1.25, parameterName=Distance Intensity Modulation Exponent}
   ))
  n11 = MaterialFunctionCall!(Width, Sharpness, Offset, Direction Switch=StaticBool() {value=true}, Compare, Input Coordinates) {materialFunction=/Engine/Functions/Engine_MaterialFunctions03/Procedurals/GeneratedBand}
  n12 = Constant()
  n13 = ScalarParameter() {group=Masking, parameterName=UV Offset V}
  n14 = ScalarParameter() {defaultValue=0.5, group=Masking, parameterName=Random1}
  n15 = Multiply(
   A=Add(A=TextureCoordinate() {vTiling=0.5}, B=AppendVector(A=@n12, B=@n13))
   B=AppendVector(A=Constant() {r=1}, B=LinearInterpolate(Alpha=@n14) {constA=0.5, constB=5})
  )
  n16 = Multiply(
   A=Add(A=ScalarParameter() {defaultValue=0.25, group=Masking, parameterName=Random2}, B=@n1)
   B=Multiply(A=ScalarParameter() {defaultValue=1, parameterName=PanSpeed}, B=ScalarParameter() {defaultValue=0.1, group=Masking, parameterName=Time Mult})
  )
  n17 = Multiply(A=TextureSampleParameter2D(UVs=Panner(Coordinate=@n15, Time=@n16) {speedY=-1}) {parameterName=Texture, texture=/Game/Effects/Fort_Effects/Textures/Tile/T_Lightning_Plasma_Soft}.R) {constB=4}
  Multiply(
   A=ParticleColor().A
   B=Multiply(A=Multiply(A=@n10.PS, B=@n8.Sinusoidal), B=StaticSwitchParameter(True=@n11.Result, False=@n17) {parameterName=Use Band?})
  )
  n18 = Time()
  n19 = ScalarParameter() {defaultValue=1, desc=Jiggle WPO Speed, parameterName=WPO Noise Speed}
  n20 = Subtract(A=WorldPosition().XYZ, B=ParticlePositionWS())
  n21 = Multiply(A=ScalarParameter() {defaultValue=5, parameterName=WPO WaveScale}, B=@n2.Noise Intensity)
  n22 = Sine(None=Add(A=Multiply(A=@n18, B=@n19), B=Divide(A=@n20, B=@n21) {constB=1000})) {period=2}
  n23 = Multiply(A=ScalarParameter() {defaultValue=1, parameterName=WPO SmallWaveScale}, B=@n2.Noise Intensity)
  Reroute(None=Add(
    A=Multiply(A=@n22, B=@n23)
    B=MaterialFunctionCall(WidthBase=@n9, WidthTip=@n9) {materialFunction=&SplineThicken}.WorldPositionOffset
   ))
 ]
}
